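# Control Rig for the Carlos male character's face accessory part that adjusts hat and back-hair bones as the head bends.
# A sequence first runs a spherical pose reader on the head relative to spine_04, storing its output in neck_bend_alpha.
# It then uses that alpha to weight a Modify Transforms on base bones such as base_front_hat and base_back_hair_b_01.
# The pose reader has debug drawing enabled.
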ControlRigBlueprint /BRCosmetics/Characters/Player/Male/Medium/Bodies/M_MED_Carlos/Meshes/Parts/M_MED_Carlos_FaceAcc_CtrlRig
graphs [
 RigVMFunctionLibrary {links [], nodes []}
 RigVMModel {nodes [
   RigUnit_BeginExecution = &RigUnit_BeginExecution()
   RigVMFunction_Sequence = &RigVMFunction_Sequence(ExecuteContext=@RigUnit_BeginExecution.ExecutePin)
   n1 = &RigUnit_SphericalPoseReader(
    ActiveRegionScaleFactors=(PositiveWidth=3,NegativeWidth=0.1,PositiveHeight=6,NegativeHeight=6)
    ActiveRegionSize=0.1
    Debug=(bDrawDebug=True,DebugScale=25,DebugSegments=20,DebugThickness=0.25)
    DriverAxis=(X=1,Y=0,Z=0)
    DriverItem=(Type=Bone,Name="head")
    FalloffRegionScaleFactors=(PositiveWidth=-3,NegativeWidth=1,PositiveHeight=0,NegativeHeight=0)
    FalloffSize=0.25
    OptionalParentItem=(Type=Bone,Name="spine_04")
    RotationOffset=(X=90,Y=0,Z=-180)
    ExecutePin=@RigVMFunction_Sequence.A
   )
   > RigVMVariableNode(Value=@n1.OutputParam) {variable=neck_bend_alpha}
   &RigUnit_ModifyTransforms(
    ItemToModify="((Item=(Type=Bone,Name=\"base_front_hat\"),Transform=(Rotation=(X=-0.0868,Y=-0.4924,Z=-0.1504,W=0.8529),Translation=(X=0,Y=0,Z=0),Scale3D=(X=1,Y=1,Z=1))),(Item=(Type=Bone,Name=\"base_back_hair_b_01\"),Transform=(Rotation=(X=0,Y=0,Z=0,W=1),Translation=(X=0,Y=0, ...(truncated)"
    ExecutePin=@RigVMFunction_Sequence.B
    Weight=RigVMVariableNode() {variable=neck_bend_alpha}.Value
   )
  ]}
]
hierarchy [
 root {loc=<0 0 0>, rot=<0 0 0 1>}
 pelvis {parent=root, loc=<0 -0.7071 93.1191>, rot=<0.0127 0.707 -0.0127 -0.707>}
 spine_01 {parent=pelvis, loc=<6.7494 -0.6388 0>, rot=<0 0 0.034 -0.9994>}
 spine_02 {parent=spine_01, loc=<7.2518 -0.035 0>, rot=<0 0 0.0238 -0.9997>}
 spine_03 {parent=spine_02, loc=<6.0104 0 0>, rot=<0 0 0.0149 -0.9999>}
 spine_04 {parent=spine_03, loc=<8.1024 0 0>, rot=<0 0 -0.0132 -0.9999>}
 spine_05 {parent=spine_04, loc=<10.0146 0 0>, rot=<0 0 -0.0015 -1>}
 neck_01 {parent=spine_05, loc=<11.3138 0.1529 0>, rot=<0 0 -0.2003 -0.9797>}
 neck_02 {parent=neck_01, loc=<4.5627 0 0>, rot=<0 0 0.0504 -0.9987>}
 head {parent=neck_02, loc=<5.3914 0 0>, rot=<0 0 0.115 -0.9934>}
 dyn_EarRing {parent=head, loc=<0.8471 0.7658 5.9252>, rot=<0.0786 0.4974 0.8629 0.0429>}
 dyn_side_Hair_a_l_01 {parent=head, loc=<6.8084 8.3504 -4.0372>, rot=<0.0361 -0.9374 0.0891 -0.3346>}
 dyn_side_Hair_a_l_02 {parent=dyn_side_Hair_a_l_01, loc=<2.822 -0.1775 0.2057>, rot=<-0.1052 0.1786 0.124 -0.9704>}
 dyn_side_Hair_b_l_01 {parent=head, loc=<3.8774 5.1824 -5.5209>, rot=<0.2385 -0.755 0.3594 -0.4938>}
 dyn_side_Hair_b_l_02 {parent=dyn_side_Hair_b_l_01, loc=<2.2713 0.0295 -0.3254>, rot=<-0.0346 0.6175 -0.1318 -0.7747>}
 dyn_side_Hair_c_r_01 {parent=head, loc=<5.9535 6.4846 5.2656>, rot=<0.5869 -0.2255 0.7567 -0.1794>}
 base_back_hair_a_01 {parent=head, loc=<-1.7955 -6.7187 0.7432>, rot=<0.5021 -0.4979 0.4979 -0.5021>}
 dyn_back_hair_a_01 {parent=base_back_hair_a_01, loc=<0 0 0>, rot=<0 0 0 1>}
 base_back_hair_b_01 {parent=head, loc=<-1.5411 -4.9285 -4.1962>, rot=<0.3431 -0.6183 0.338 -0.6211>}
 dyn_back_hair_b_01 {parent=base_back_hair_b_01, loc=<0 0 0>, rot=<0 0 0 1>}
 base_back_hair_c_01 {parent=head, loc=<-1.1355 -3.3787 5.6116>, rot=<0.6984 -0.1109 0.6974 -0.1167>}
 dyn_back_hair_c_01 {parent=base_back_hair_c_01, loc=<0 0 0>, rot=<0 0 0 1>}
 base_front_hat {parent=head, loc=<4.5384 -6.686 0.0082>, rot=<0.4164 -0.5716 0.5712 -0.4167>}
]
variables [
 Neck_bend {type=double, default=0}
 Targeting_alpha {type=double, default=0}
 neck_bend_alpha {type=double, default=0}
]
